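# Physics asset for the face accessory mesh of the female medium Tie Dye Fashion player cosmetic, defining hair dynamics.
# Front, side and back hair form two-box chains from the head, with smaller strand boxes hanging off their tips or the head.
# Capsules cover head, neck and upper spine.
# Hair joints have swing limits and tight twist, while body joints and a side-to-back tip cross-link leave rotation free.
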PhysicsAsset /BRCosmetics/Characters/Player/Female/Medium/Bodies/F_MED_Tie_Dye_Fashion/Meshes/Parts/A/F_MED_TieDyeFashion_A_FaceAcc_Physics
bodies {
 dyn_back_hair_a [{shapeType=Box, extentX=5.5, extentY=10, extentZ=3.5, center=<2.688 0 0>}]
 dyn_back_hair_b [{shapeType=Box, extentX=2, extentY=8, extentZ=3.5, center=<1.1385 0 0>}]
 dyn_front_hair_a [{shapeType=Box, extentX=5, extentY=10.5, extentZ=1.5, center=<2.3959 0 0>, rotation=<rot 0 0 12.658>}]
 dyn_front_hair_b [{shapeType=Box, extentX=4, extentY=9, extentZ=1.5, center=<2.2086 0 0>, rotation=<rot 0 0 17.1016>}]
 dyn_hair_01 [{shapeType=Box, extentX=5, extentY=5, extentZ=1.5, center=<2.5 0 0>}]
 dyn_hair_02 [{shapeType=Box, extentX=2.5, extentY=1.5, extentZ=4, center=<1.2655 0 -1.5081>, rotation=<rot 40 0 0>}]
 dyn_hair_03 [{shapeType=Box, extentX=5, extentY=5, extentZ=2, center=<2.5 0 0>}]
 dyn_hair_04 [{shapeType=Box, extentX=5, extentY=5, extentZ=2.5, center=<2.5 0 0>}]
 dyn_hair_05 [{shapeType=Box, extentX=5, extentY=5, extentZ=1.5, center=<2.5 0 0>}]
 dyn_hair_06 [{shapeType=Box, extentX=4.5, extentY=3.5, extentZ=2.5, center=<0.4 0 0>}]
 dyn_hair_07 [{shapeType=Box, extentX=5, extentY=2, extentZ=2, center=<2.5 0 0>}]
 dyn_hair_08 [{shapeType=Box, extentX=5, extentY=2, extentZ=2, center=<2.5 0 0>}]
 dyn_hair_09 [{shapeType=Box, extentX=4.5, extentY=5, extentZ=2, center=<1.6615 0 0>}]
 dyn_side_hair_a [{shapeType=Box, extentX=6, extentY=9.5, extentZ=3.5, center=<2.9675 0 0>}]
 dyn_side_hair_b [{shapeType=Box, extentX=5, extentY=8.5, extentZ=3, center=<2.5085 0 0>}]
 head [{shapeType=Capsule, length=6, radius=6.5, center=<2.8432 0 0>, rotation=<rot 90 0 0>}]
 neck_01 [{shapeType=Capsule, length=6.5, radius=4.25, center=<2.081 0.2413 0>, rotation=<rot 90 -90 -87.8375>}]
 spine_05 [{shapeType=Capsule, length=10, radius=9, center=<5 0.5 0>}]
}
constraints [
 {bone1Name=dyn_hair_06, bone2Name=head, swing1LimitDegrees=45, swing2LimitDegrees=45, twistLimitDegrees=0.5}
 {bone1Name=dyn_hair_05, bone2Name=dyn_front_hair_b, swing1LimitDegrees=45, swing2LimitDegrees=45, twistLimitDegrees=0.5}
 {bone1Name=dyn_hair_04, bone2Name=dyn_front_hair_b, swing1LimitDegrees=45, swing2LimitDegrees=45, twistLimitDegrees=0.5}
 {bone1Name=dyn_hair_03, bone2Name=dyn_front_hair_b, swing1LimitDegrees=45, swing2LimitDegrees=45, twistLimitDegrees=0.5}
 {bone1Name=dyn_hair_02, bone2Name=dyn_side_hair_b, swing1LimitDegrees=45, swing2LimitDegrees=45, twistLimitDegrees=0.5}
 {bone1Name=dyn_hair_01, bone2Name=head, swing1LimitDegrees=45, swing2LimitDegrees=45, twistLimitDegrees=0.5}
 {bone1Name=dyn_hair_09, bone2Name=dyn_back_hair_b, swing1LimitDegrees=45, swing2LimitDegrees=45, twistLimitDegrees=5}
 {bone1Name=dyn_hair_08, bone2Name=head, swing1LimitDegrees=40, swing2LimitDegrees=20, twistLimitDegrees=5}
 {bone1Name=dyn_hair_07, bone2Name=head, swing1LimitDegrees=40, swing2LimitDegrees=20, twistLimitDegrees=5}
 {bone1Name=dyn_front_hair_a, bone2Name=head, swing1LimitDegrees=30, swing2LimitDegrees=30, twistLimitDegrees=5}
 {bone1Name=dyn_side_hair_a, bone2Name=head, swing1LimitDegrees=30, swing2LimitDegrees=30, twistLimitDegrees=5}
 {bone1Name=dyn_back_hair_a, bone2Name=head, swing1LimitDegrees=30, swing2LimitDegrees=30, twistLimitDegrees=5}
 {bone1Name=dyn_front_hair_b, bone2Name=dyn_front_hair_a, swing1LimitDegrees=30, swing2LimitDegrees=30, twistLimitDegrees=5}
 {bone1Name=dyn_side_hair_b, bone2Name=dyn_side_hair_a, swing1LimitDegrees=30, swing2LimitDegrees=30, twistLimitDegrees=5}
 {bone1Name=dyn_back_hair_b, bone2Name=dyn_back_hair_a, swing1LimitDegrees=30, swing2LimitDegrees=30, twistLimitDegrees=5}
 {
  bone1Name=head
  bone2Name=neck_01
  swing1LimitDegrees=45
  swing1Motion=Free
  swing2LimitDegrees=45
  swing2Motion=Free
  twistLimitDegrees=45
  twistMotion=Free
 }
 {
  bone1Name=neck_01
  bone2Name=spine_05
  swing1LimitDegrees=45
  swing1Motion=Free
  swing2LimitDegrees=45
  swing2Motion=Free
  twistLimitDegrees=45
  twistMotion=Free
 }
 {
  bone1Name=dyn_side_hair_b
  bone2Name=dyn_back_hair_b
  swing1LimitDegrees=45
  swing1Motion=Free
  swing2LimitDegrees=45
  swing2Motion=Free
  twistLimitDegrees=45
  twistMotion=Free
 }
]
