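# Animation Blueprint for the hair bones of the female Tie Dye Fashion face-accessory mesh.
# It copies the attached parent's pose, simulates RigidBody physics in neck_01 space with scaled component motion, then stacks additive bone offsets to reposition hair strands.
# The update event derives head, clavicle and crouch-targeting values, though no AnimGraph node visibly reads them.
# Two hair-to-head constraints and a head transform sit disconnected from the output.
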AnimBlueprint /BRCosmetics/Characters/Player/Female/Medium/Bodies/F_MED_Tie_Dye_Fashion/Meshes/Parts/A/F_MED_TieDyeFashion_A_FaceAcc_Skeleton_AnimBP
target_skeleton=&/F_MED_TieDyeFashion_A_FaceAcc_Skeleton
graphs [
 AnimGraph {nodes [
   n1 = "Constraint - dyn_side_hair_a (P)"(Alpha=1, ConstraintWeights_0=1) {
    alphaBoolBlend {blendOption=Linear}
    boneToModify {boneName=dyn_side_hair_a}
    constraintSetup [{targetBone {boneName=head}}]
    constraintWeights [1]
   }
   "Constraint - dyn_back_hair_a (P)"(ComponentPose=@n1, Alpha=1, ConstraintWeights_0=1) {
    alphaBoolBlend {blendOption=Linear}
    boneToModify {boneName=dyn_back_hair_a}
    constraintSetup [{targetBone {boneName=head}}]
    constraintWeights [1]
   }
   "Transform (Modify) Bone - Bone: head"(Alpha=1, Translation="0,0,0", Rotation="0,0,0", Scale="1,1,1") {translationMode=BMM_Additive, alphaBoolBlend {blendOption=Linear}, boneToModify {boneName=head}}
   n2 = Get mesh_to_copy() -> (mesh_to_copy:Skeletal Mesh Component Object Reference)
   n3 = RigidBody(ComponentPose=Local To Component(LocalPose=Copy Pose From Mesh(SourceMeshComponent=@n2) {bUseAttachedParent=true})) {
    bClampLinearTranslationLimitToRefPose=true
    bOverrideWorldGravity=true
    overridePhysicsAsset=&/F_MED_TieDyeFashion_A_FaceAcc_Physics
    simulationSpace=BaseBoneSpace
    alphaBoolBlend {blendOption=Linear}
    baseBoneRef {boneName=neck_01}
    componentAppliedLinearAccClamp=<3000 3000 3000>
    componentLinearAccScale=<0.3 0.25 0.25>
    componentLinearVelScale=<0.3 0.25 0.25>
    overrideWorldGravity {z=-980}
   }
   n4 = "Transform (Modify) Bone - Bone: dyn_side_hair_a"(ComponentPose=@n3, Alpha=1, Translation="-2,0,0", Rotation="10,-5,0", Scale="1,1,1") {
    rotationMode=BMM_Additive
    rotationSpace=BCS_BoneSpace
    translationMode=BMM_Additive
    translationSpace=BCS_BoneSpace
    alphaBoolBlend {blendOption=Linear}
    boneToModify {boneName=dyn_side_hair_a}
    rotation {pitch=10, yaw=-5}
    translation {x=-2}
   }
   n5 = "Transform (Modify) Bone - Bone: dyn_back_hair_a"(ComponentPose=@n4, Alpha=1, Translation="-2,0,0", Rotation="10,5,0", Scale="1,1,1") {
    rotationMode=BMM_Additive
    rotationSpace=BCS_BoneSpace
    translationMode=BMM_Additive
    translationSpace=BCS_BoneSpace
    alphaBoolBlend {blendOption=Linear}
    boneToModify {boneName=dyn_back_hair_a}
    rotation {pitch=10, yaw=5}
    translation {x=-2}
   }
   n6 = "Transform (Modify) Bone - Bone: dyn_side_hair_b"(ComponentPose=@n5, Alpha=1, Translation="-2,0,0", Rotation="0,0,0", Scale="1,1,1") {
    translationMode=BMM_Additive
    translationSpace=BCS_BoneSpace
    alphaBoolBlend {blendOption=Linear}
    boneToModify {boneName=dyn_side_hair_b}
    translation {x=-2}
   }
   n7 = "Transform (Modify) Bone - Bone: dyn_hair_02"(ComponentPose=@n6, Alpha=1, Translation="-2,0,0", Rotation="0,0,0", Scale="1,1,1") {
    translationMode=BMM_Additive
    translationSpace=BCS_BoneSpace
    alphaBoolBlend {blendOption=Linear}
    boneToModify {boneName=dyn_hair_02}
    translation {x=-2}
   }
   n8 = "Transform (Modify) Bone - Bone: dyn_back_hair_b"(ComponentPose=@n7, Alpha=1, Translation="-2,0,0", Rotation="0,0,0", Scale="1,1,1") {
    translationMode=BMM_Additive
    translationSpace=BCS_BoneSpace
    alphaBoolBlend {blendOption=Linear}
    boneToModify {boneName=dyn_back_hair_b}
    translation {x=-2}
   }
   n9 = "Transform (Modify) Bone - Bone: dyn_hair_09"(ComponentPose=@n8, Alpha=1, Translation="-2,0,0", Rotation="0,0,0", Scale="1,1,1") {
    translationMode=BMM_Additive
    translationSpace=BCS_BoneSpace
    alphaBoolBlend {blendOption=Linear}
    boneToModify {boneName=dyn_hair_09}
    translation {x=-2}
   }
   n10 = "Transform (Modify) Bone - Bone: dyn_side_hair_b"(ComponentPose=@n9, Alpha=1, Translation="2,0,0", Rotation="0,0,0", Scale="1,1,1") {
    translationMode=BMM_Additive
    translationSpace=BCS_BoneSpace
    alphaBoolBlend {blendOption=Linear}
    boneToModify {boneName=dyn_side_hair_b}
    translation {x=2}
   }
   n11 = "Transform (Modify) Bone - Bone: dyn_hair_02"(ComponentPose=@n10, Alpha=1, Translation="2,0,0", Rotation="0,0,0", Scale="1,1,1") {
    translationMode=BMM_Additive
    translationSpace=BCS_BoneSpace
    alphaBoolBlend {blendOption=Linear}
    boneToModify {boneName=dyn_hair_02}
    translation {x=2}
   }
   n12 = "Transform (Modify) Bone - Bone: dyn_back_hair_b"(ComponentPose=@n11, Alpha=1, Translation="-2.5,0,0", Rotation="0,0,0", Scale="1,1,1") {
    translationMode=BMM_Additive
    translationSpace=BCS_BoneSpace
    alphaBoolBlend {blendOption=Linear}
    boneToModify {boneName=dyn_back_hair_b}
    translation {x=-2.5}
   }
   n13 = "Transform (Modify) Bone - Bone: dyn_hair_09"(ComponentPose=@n12, Alpha=1, Translation="-2.5,0,0", Rotation="0,0,0", Scale="1,1,1") {
    translationMode=BMM_Additive
    translationSpace=BCS_BoneSpace
    alphaBoolBlend {blendOption=Linear}
    boneToModify {boneName=dyn_hair_09}
    translation {x=-2.5}
   }
   n14 = "Transform (Modify) Bone - Bone: dyn_side_hair_b"(ComponentPose=@n13, Alpha=1, Translation="0,0,1.5", Rotation="0,0,0", Scale="1,1,1") {
    translationMode=BMM_Additive
    translationSpace=BCS_BoneSpace
    alphaBoolBlend {blendOption=Linear}
    boneToModify {boneName=dyn_side_hair_b}
    translation {z=1.5}
   }
   n15 = "Transform (Modify) Bone - Bone: dyn_side_hair_a"(ComponentPose=@n14, Alpha=1, Translation="0,0,0", Rotation="20,0,0", Scale="1,1,1") {
    rotationMode=BMM_Additive
    rotationSpace=BCS_BoneSpace
    alphaBoolBlend {blendOption=Linear}
    boneToModify {boneName=dyn_side_hair_a}
    rotation {pitch=20}
   }
   n16 = "Transform (Modify) Bone - Bone: dyn_hair_02"(ComponentPose=@n15, Alpha=1, Translation="-0.000,-2,0", Rotation="0,0,0", Scale="1,1,1") {
    rotationMode=BMM_Additive
    rotationSpace=BCS_BoneSpace
    translationMode=BMM_Additive
    translationSpace=BCS_BoneSpace
    alphaBoolBlend {blendOption=Linear}
    boneToModify {boneName=dyn_hair_02}
    translation {y=-2}
   }
   n17 = "Transform (Modify) Bone - Bone: dyn_side_hair_b"!(ComponentPose=@n16, bAlphaBoolEnabled:bool=True, Translation="0,0,1", Rotation="0,0,0", Scale="1,1,1") -> (Pose) {
    alphaInputType=Bool
    rotationMode=BMM_Additive
    rotationSpace=BCS_BoneSpace
    translationMode=BMM_Additive
    translationSpace=BCS_BoneSpace
    alphaBoolBlend {blendInTime=0.1, blendOption=Linear, blendOutTime=0.5}
    boneToModify {boneName=dyn_side_hair_b}
    translation {z=1}
   }
   Output Pose(Result=Component To Local(ComponentPose=@n17)) {layerGroup=None}
  ]}
 EventGraph {nodes [
   Event BlueprintUpdateAnimation()
   n1 = GetOwnerSkeletalMeshComponent()
   n2 = BreakTransform(InTransform=GetDeltaTransformFromRefPose(self=@n1, BoneName=head, BaseName=spine_05))
   > Set headBack(headBack:double=MapRangeClamped(Value=BreakRotator(InRot=@n2.Rotation).Yaw, InRangeA=-10, InRangeB=-40)) -> (Output_Get:double)
   n3 = BreakRotator(InRot=BreakTransform(InTransform=GetDeltaTransformFromRefPose(self=@n1, BoneName=clavicle_l, BaseName=spine_05)).Rotation)
   > Set clavicleDelta(clavicleDelta:double=MapRangeClamped(Value=@n3.Pitch, InRangeA=0, InRangeB=45)) -> (Output_Get:double)
   > Set clavicle_delta_targeting(clavicle_delta_targeting:double=MapRangeClamped(Value=@n3.Pitch, InRangeA=13, InRangeB=19)) -> (Output_Get:double)
   n4 = BreakTransform(InTransform=GetDeltaTransformFromRefPose(self=@n1, BoneName=head, BaseName=spine_01))
   > Set headForward(headForward:double=MapRangeClamped(Value=BreakRotator(InRot=@n4.Rotation).Yaw, InRangeA=-10, InRangeB=20)) -> (Output_Get:double)
   > Set Crouch_targeting(Crouch_targeting:bool=AND Boolean(A=Get bIsCrouching(), B=Get bIsTargeting())) -> (Output_Get:bool)
   Event BlueprintInitializeAnimation()
   > Set mesh_to_copy(mesh_to_copy:Skeletal Mesh Component Object Reference=GetOwnerSkeletalMeshComponent()) -> (Output_Get:Skeletal Mesh Component Object Reference)
   TryGetPawnOwner()
  ]}
]
variables [
 clavicle_delta_targeting {type=number}
 clavicleDelta {type=number}
 Crouch_targeting {type=boolean}
 headBack {type=number}
 headForward {type=number}
 mesh_to_copy {type=SkeletalMeshComponent}
]
